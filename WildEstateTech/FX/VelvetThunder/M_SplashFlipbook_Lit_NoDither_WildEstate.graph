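# Masked, Default Lit splash flipbook material for Niagara sprites, with responsive AA enabled and, per its name, no dither.
# Base texture green, raised to a power and tinted by particle color, forms the color, with a scaled copy for emissive.
# Dynamic parameter MaskErode exponentiates that green channel, then texture alpha, particle alpha and MaskScale multiply it into an erodable mask.
# A water-splash normal map is flattened by dynamic parameter FlattenNormal.
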Material /WildEstateTech/FX/VelvetThunder/M_SplashFlipbook_Lit_NoDither_WildEstate
graph {
 properties {
  bEnableResponsiveAA=true
  bUsedWithNiagaraSprites=true
  blendMode=BLEND_Masked
  pixelDepthOffsetMode=PDOM_Legacy
  usedShadingModels=Default Lit
 }
 nodes [
  n1 = TextureSampleParameter2D() {parameterName=Texture_Normal, samplerType=SAMPLERTYPE_Normal, texture=/Game/Effects/Fort_Effects/Textures/Water/T_waterSplash_R_N}
  MaterialFunctionCall(Normal=@n1.RGB, Flatness=DynamicParameter() {paramNames [MaskErode, FlattenNormal, Param3, Param4]}.FlattenNormal) {materialFunction=&FlattenNormal}
  n2 = TextureSampleParameter2D() {parameterName=Texture_Base}
  n3 = Power(Base=@n2.G, Exp=DynamicParameter() {paramNames [MaskErode, FlattenNormal, Param3, Param4]}.MaskErode)
  n4 = ParticleColor()
  Multiply(A=Multiply(A=@n4.A, B=Multiply(A=@n3, B=@n2.A)), B=ScalarParameter() {defaultValue=0.5, parameterName=MaskScale})
  n5 = Multiply(A=@n4.RGB, B=Power(Base=@n2.G, Exp=ScalarParameter() {defaultValue=0.2, parameterName=FlipBook_Power}))
  Multiply(A=@n5, B=ScalarParameter() {parameterName=EmissiveContribution})
  Reroute(None=@n5)
  ScalarParameter() {parameterName=Specular}
  ScalarParameter() {parameterName=Roughness}
  ScalarParameter() {parameterName=Metallic}
 ]
}
