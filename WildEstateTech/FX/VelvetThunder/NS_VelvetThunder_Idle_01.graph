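# Niagara idle effect for VelvetThunder, combining cold smoke with orange drips.
# ColdSmoke continuously spawns faint, slowly sinking flipbook sprites in a raised box, fading them in and out and tinting brightness by time of day.
# Drips bursts probabilistically from two fixed offsets, with count driven by User.SpawnDrips.
# Each drip grows while curve-driven high drag holds it in place, then falls under gravity once drag drops, with a kill volume below.
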NiagaraSystem /WildEstateTech/FX/VelvetThunder/NS_VelvetThunder_Idle_01
dependencies {
 usedDataInterfaces [
  &NiagaraDataInterfaceArrayFloat3
  &NiagaraDataInterfaceCurve
  &NiagaraDataInterfaceVector2DCurve
  &NiagaraDataInterfaceSpriteRendererInfo
 ]
 usedDynamicInputs [&Lerp_Float]
 usedModules [&SystemState]
}
emitters {
 ColdSmoke {
  emitterUpdateScript [
   &EmitterState()
   &SpawnRate(SpawnRate=8, Use Spawn Probability=false)
  ]
  particleSpawnScript [
   &InitializeParticle(
    Color Maximum=<rgba 0.9 0.9667 1 0.1>
    Color Minimum=<rgba 0.9 0.9667 1 0.1>
    Color Mode=::NewEnumerator2
    Lifetime=&RandomRangeFloat(Minimum=4.5, Maximum=4.5)
    Sprite Rotation Angle=&Multiply_Float(A=66, B=&RandomRangeFloat(Minimum=-1, Maximum=1))
    Sprite Rotation Mode=::NewEnumerator0
    Sprite Size=&Multiply_Vector2D_ByFloat(Float=100, Vector2D=&RandomRangeVector2D(Minimum=<0.8 0.8>, Maximum=<1 1>))
    Sprite Size Mode=::NewEnumerator4
    UsePositionOffset=false
    Write Material Random=true
   )
   &ShapeLocation(
    Box Size=<100 50 25>
    MaskSpawnGroup=false
    Offset=<0 0 190>
    Offset Mode=::NewEnumerator0
    Override Seed=false
    Shape Primitive=::NewEnumerator2
    Surface Only=false
    UseOwnerScale=true
   )
   &CameraOffset(Camera Offset Amount=&RandomRangeFloat(Minimum=20, Maximum=30))
  ]
  particleUpdateScript [
   &ParticleState()
   &ScaleSpriteSize(Uniform Curve Sprite Scale=DataInterface() {propertyValues {Curve {
       defaultValue=0
       keys [
        {
         arriveTangentWeight=0.3283
         interpMode=RCIM_Cubic
         leaveTangentWeight=0.3283
         tangentMode=RCTM_User
         tangentWeightMode=RCTWM_WeightedBoth
         time=0
         value=0.8
        }
        {arriveTangentWeight=0.6313, interpMode=RCIM_Cubic, tangentMode=RCTM_Break, tangentWeightMode=RCTWM_WeightedBoth, time=1, value=1.1}
       ]
      }}})
   &SpriteRotationRate(Rotation Rate=&Multiply_Float(A=18, B=&RandomRangeFloat(Minimum=-1, Maximum=1)))
   &GravityForce(Gravity=&Multiply_VectorByFloat(Vector=<0 0 -980>, Float=0.015))
   &SolveForcesAndVelocity(Clamp Velocity=false, Limit Acceleration=false)
   &ScaleColor(
    Scale Alpha=&FloatFromCurve(FloatCurve=DataInterface() {propertyValues {Curve {
        defaultValue=0
        keys [
         <at 0 0 RCIM_Cubic>
         <at 0.45 1 RCIM_Cubic>
         {arriveTangent=-0.0405, arriveTangentWeight=0.1833, interpMode=RCIM_Cubic, leaveTangent=-0.0405, tangentMode=RCTM_User, time=1, value=0}
        ]
       }}})
    Scale RGB=&FN_Select_Vector_from_Time_Of_Day(Morning=<0.875 0.875 0.875>, Day=<1 1 1>, Evening=<0.875 0.875 0.875>, Night=<0.75 0.75 0.75>)
    ScaleA=true
    ScaleRGB=true
   )
   &DynamicMaterialParameters(
    Index 0 Param 2=50
    Index 0 Param 3=1
    Module.Index0Float0Write=true
    Module.Index0Float1Write=true
    Module.Index0Float2Write=true
    Module.Index0Float3Write=true
   )
   &SubUVAnimation(
    "Sprite Renderer "=DataInterface() {propertyValues {SpriteRenderer=&self:ColdSmoke_0.NiagaraSpriteRendererProperties_2}}
    Start Frame Offset:NiagaraInt32=&RandomRangeInt(Minimum=0, Maximum=23)
    SubUV Lookup Index:NiagaraFloat=&FloatFromCurve(
     FloatCurve=DataInterface() {propertyValues {Curve {defaultValue=1, keys [{tangentMode=RCTM_User, time=0, value=0}, {tangentMode=RCTM_User, time=1, value=1}]}}}
     Scale Curve=0.75
    )
    UseEndFrame=false
    UseStartFrame=false
   )
  ]
  rendererClasses [&NiagaraSpriteRendererProperties]
  renderers [{rendererClass=&NiagaraSpriteRendererProperties}]
 }
 Drips {
  emitterUpdateScript [
   &EmitterState()
   &SpawnBurst_Instantaneous(
    Spawn Count=&Multiply_Int(A=1, B=$User.SpawnDrips)
    Spawn Probability=0.75
    Use Spawn Probability=true
    UseLoopCountLimit=false
   )
  ]
  particleSpawnScript [
   &InitializeParticle(
    Color=<rgba 1.25 0.5208 0 2>
    Lifetime=1.5
    Sprite Size Max=<5 18>
    Sprite Size Min=<4 12>
    Sprite Size Mode=::NewEnumerator2
    UsePositionOffset=false
    Write Material Random=true
   )
   &SystemLocation(
    MaskSpawnGroup=false
    Offset=&SelectVectorFromArray(
     Array Sampling Mode=::NewEnumerator0
     Direct Array Index=?
     Vector Selection Array=DataInterface() {propertyValues {FloatData [<-25.65 57 107>, <25.65 57 107>]}}
     Random Seed=0
    )
   )
  ]
  particleUpdateScript [
   &ParticleState()
   &ScaleSpriteSize(
    Non-Uniform Curve Sprite Scale=DataInterface() {propertyValues {
      XCurve {keys [<at 0 0 RCIM_Cubic>, {arriveTangent=7.3849, tangentMode=RCTM_User, time=0.2, value=0.7626}, {tangentMode=RCTM_User, time=1, value=1}]}
      YCurve {keys [<at 0 0 RCIM_Cubic>, {arriveTangent=4.3694, tangentMode=RCTM_User, time=0.2, value=0.5253}, {tangentMode=RCTM_User, time=1, value=1}]}
     }}
    Scale Sprite Size Mode=::NewEnumerator3
   )
   &GravityForce(Gravity=&Multiply_VectorByFloat(Vector=<0 0 -980>, Float=0.5))
   &Drag(
    Drag=&FloatFromCurve(
     FloatCurve=DataInterface() {propertyValues {Curve {keys [
         {arriveTangent=-2.875, interpMode=RCIM_Constant, leaveTangentWeight=1.3155, time=0, value=1}
         {
          arriveTangentWeight=1.1296
          interpMode=RCIM_Cubic
          leaveTangentWeight=1.1296
          tangentMode=RCTM_User
          tangentWeightMode=RCTWM_WeightedBoth
          time=0.328
          value=0
         }
        ]}}}
     Scale Curve=1000
    )
    Use Linear Drag=true
    Use Rotational Drag=true
   )
   &SolveForcesAndVelocity(Clamp Velocity=false, Limit Acceleration=false)
   &KillParticlesInVolume(Kill Shape=::NewEnumerator2, Origin Offset:Vector3f=<0 0 -50>)
  ]
  rendererClasses [&NiagaraSpriteRendererProperties]
  renderers [{rendererClass=&NiagaraSpriteRendererProperties}]
 }
}
summary {
 systemName=NS_VelvetThunder_Idle_01
 emitters [
  {emitterName=Drips, rendererClasses [&NiagaraSpriteRendererProperties]}
  {emitterName=ColdSmoke, rendererClasses [&NiagaraSpriteRendererProperties]}
 ]
 userVariables [
  User.Drips_On {type=NiagaraInt32}
  User.SpawnDrips {type=NiagaraInt32, defaultValue=1}
 ]
}
